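# Conversation database for the Carolina Chili NPC's dialogue, flavor chat with chili-pepper puns.
# The tagged entry leads to an "Ask about..." greeting that branches into a "Seen anything strange?" topic or a Back node.
# The topic plays four chained Speech lines about a wandering stranger, then reaches a ReturnToStart node with a BACK choice.
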ConversationDatabase /NPCLibrary/NPCs/CarolinaChilli/NPCLibrary_Conversation_CarolinaChili_Dialogue
entry_points [{tag=Conversation.Entry.CarolinaChili.Dialogue, destinations [#1]}]
nodes [
 ConversationEntryPointNode_0 {
  id=#1
  class=ConversationEntryPointNode
  outputs [#2]
  properties {entryTag {tagName=Conversation.Entry.CarolinaChili.Dialogue}}
 }
 Speech_C_0 {
  id=#2
  class=Speech_C
  outputs [#3, #4]
  properties {
   nodeName=""
   softTaskIcon=&T_UI_NPC_question
   generalConfig {defaultMessage=I'm here to answer your burning questions.}
  }
  sub_nodes [ConversationChoiceNode_5 {class=ConversationChoiceNode, properties {defaultChoiceDisplayText=Ask about...}}]
 }
 Back_C_0 {
  id=#4
  class=Back_C
  sub_nodes [ConversationChoiceNode_7 {class=ConversationChoiceNode, properties {defaultChoiceDisplayText=BACK}}]
 }
 Speech_C_7 {
  id=#3
  class=Speech_C
  outputs [#5]
  properties {
   nodeName=""
   softTaskIcon=&T_UI_NPC_question
   generalConfig {defaultMessage=A wanderer's been walking the wasteland.}
  }
  sub_nodes [ConversationChoiceNode_15 {class=ConversationChoiceNode, properties {defaultChoiceDisplayText=Seen anything strange?}}]
 }
 Speech_C_15 {
  id=#5
  class=Speech_C
  outputs [#6]
  properties {nodeName="", softTaskIcon=&T_UI_NPC_question, generalConfig {defaultMessage=Looks like one spicy character.}}
 }
 Speech_C_14 {
  id=#6
  class=Speech_C
  outputs [#7]
  properties {
   nodeName=""
   softTaskIcon=&T_UI_NPC_question
   generalConfig {defaultMessage="Vanishes faster than a hot flash, though."}
  }
 }
 Speech_C_1 {
  id=#7
  class=Speech_C
  outputs [#8]
  properties {
   nodeName=""
   softTaskIcon=&T_UI_NPC_question
   generalConfig {defaultMessage=I don't think they can handle my heat.}
  }
 }
 ReturnToStart_C_23 {
  id=#8
  class=ReturnToStart_C
  properties {softTaskIcon=&T_UI_NPC_Back}
  sub_nodes [ConversationChoiceNode_32 {class=ConversationChoiceNode, properties {defaultChoiceDisplayText=BACK}}]
 }
]
